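# Niagara glow effect for the LemonWrath character intro, built from sprite emitters.
# BurstGlow spawns three aligned, elongated sprites, using the partition index to pick each one's offset and tilt and User.VariantColor to pick its color from an array.
# OnceSweep adds a brief, bright yellow sprite that shoots mostly upward.
# EndBurst and EndBurstRing are listed without renderers or modules.
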NiagaraSystem /BRCosmetics/Animation/Game/MainPlayer/CharIntros/s41/LemonWrath/FX/NS_CharIntros_LemonWrath_Glow
dependencies {
 usedDataInterfaces [
  &NiagaraDataInterfaceArrayColor
  &NiagaraDataInterfaceArrayFloat3
  &NiagaraDataInterfaceArrayFloat
 ]
 usedDynamicInputs [&SelectFloatFromArray]
 usedModules [&SystemState]
}
emitters {
 BurstGlow {
  emitterUpdateScript [
   &EmitterState(Inactive Response=::NewEnumerator1, Life Cycle Mode=::NewEnumerator1, Loop Behavior=::NewEnumerator1, UseLoopDelay=false)
   &SpawnBurst_Instantaneous(Spawn Count=3, Use Spawn Probability=false, UseLoopCountLimit=false)
  ]
  particleSpawnScript [
   &PartitionParticles(Number Of Partitions=3)
   &InitializeParticle(
    Color=&SelectLinearColorFromArray(
     Array Sampling Mode=::NewEnumerator1
     Direct Array Index=&MakeIntFromFloat(Float=$User.VariantColor)
     Color Selection Array=DataInterface() {propertyValues {ColorData [<rgba 1 0.8105 0.0798 1>, <rgba 0 0.25 1 1>, <rgba 0 0 0 1>]}}
     Random Seed=?
    )
    Lifetime=1.8
    Position Offset=&SelectVectorFromArray(
     Direct Array Index=$Particles.PartitionParticles.Partition
     Vector Selection Array=DataInterface() {propertyValues {FloatData [<0 0 25>, <0 2 40>, <0 5 55>]}}
    )
    Sprite Size=<20 80>
    Sprite Size Mode=::NewEnumerator4
    UsePositionOffset=true
    Write Material Random=true
   )
   &DynamicMaterialParameters(
    Index 0 Param 1=5
    Index 0 Param 2=5
    Module.Index0Float0Write=true
    Module.Index0Float1Write=true
    Module.Index0Float2Write=true
    Module.Index0Float3Write=true
   )
   &SpriteFacingAndAlignment(
    Sprite Alignment=&SelectVectorFromArray(
     Direct Array Index=$Particles.PartitionParticles.Partition
     Vector Selection Array=DataInterface() {propertyValues {FloatData [<0 0 1>, <0 0.2 1>, <0 0.3 1>]}}
    )
    Sprite Facing=<0 0 1>
    WriteAlignment=true
    WriteFacing=true
   )
   &CameraOffset(Camera Offset Amount=10)
  ]
  particleUpdateScript [
   &ParticleState()
   &ScaleColor(Scale Alpha=0.2, ScaleA=true, ScaleRGB=true)
  ]
  rendererClasses [&NiagaraSpriteRendererProperties]
  renderers [{rendererClass=&NiagaraSpriteRendererProperties}]
 }
 EndBurst {rendererClasses [], renderers []}
 EndBurstRing {rendererClasses [], renderers []}
 OnceSweep {
  emitterUpdateScript [
   &EmitterState(Life Cycle Mode=::NewEnumerator1, Loop Behavior=::NewEnumerator1, Loop Duration=1, UseLoopDelay=false)
   &SpawnBurst_Instantaneous(Use Spawn Probability=false, UseLoopCountLimit=false)
  ]
  particleSpawnScript [
   &InitializeParticle(Color=<rgba 5 4.5244 0.93 1>, Lifetime=0.12, UsePositionOffset=false, Write Material Random=true)
   &AddVelocity(Velocity=<0 50 600>)
   &CameraOffset(Camera Offset Amount=5)
  ]
  particleUpdateScript [
   &ParticleState()
   &SolveForcesAndVelocity(Clamp Velocity=false, Limit Acceleration=false)
  ]
  rendererClasses [&NiagaraSpriteRendererProperties]
  renderers [{rendererClass=&NiagaraSpriteRendererProperties}]
 }
}
summary {
 systemName=NS_CharIntros_LemonWrath_Glow
 emitters [
  {emitterName=BurstGlow, rendererClasses [&NiagaraSpriteRendererProperties]}
  {emitterName=OnceSweep, rendererClasses [&NiagaraSpriteRendererProperties]}
  {emitterName=EndBurst, rendererClasses []}
  {emitterName=EndBurstRing, rendererClasses []}
 ]
 userVariables [
  User.bIsFrontEnd {type=NiagaraBool}
  User.bIsFrontEndPreview {type=NiagaraBool}
  User.RibbonIDOffset {type=NiagaraInt32}
  User.SwingSectionIndex {type=NiagaraInt32}
  User.VariantColor {type=NiagaraFloat, defaultValue=0, description=""}
 ]
}
